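# Material function for the HealedScar Sand backpack skin that layers glowing, animated wisps and Fresnel silhouette glow onto incoming material attributes.
# Wisps use triplanar projection from pre-skinned position and normal, with panned, flowmap-distorted, bump-offset noise layers on front and side.
# A packed mask drives scar emissive and surface adjustments, with static switches adding sphere-mask glow and an emissive map.
# Base color gets HSV offsets; roughness and specular are mask-adjusted.
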MaterialFunction /BRCosmetics/Accessories/FORT_Backpacks/Backpack_M_MED_HealedScar/Skins/Sand/Materials/MF_HealedScar_Sand_Assets
graph {nodes [
  Constant3Vector() {constant=<rgba 0.2235 0.949 0.8588 1>}
  n1 = GetMaterialAttributes(None=FunctionInput() {inputType=FunctionInput_MaterialAttributes}) {attributeGetTypes [
    #1
    #2
    #3
    #4
    #5
   ]}
  n2 = Multiply(A=ComponentMask(None=NamedRerouteUsage()) {r=true}) {constB=4}
  n3 = SetMaterialAttributes!(
   MaterialAttributes=@n1.MaterialAttributes
   EmissiveColor=NamedRerouteUsage()
   Normal=@n1.MaterialAttributes
   BaseColor=NamedRerouteUsage()
   Roughness=Add(A=@n1.Roughness, B=@n2)
   Specular=LinearInterpolate(A=@n1.Specular, Alpha=@n2) {constB=0}
  ) {attributeSetTypes [
    #1
    #2
    #3
    #4
    #5
   ]}
  FunctionOutput(None=@n3) {outputName=OUT}
  n4 = TextureSampleParameter2D() {
   group=HealedScar_Sand
   parameterName=HealScar_TechArt_Mask
   samplerType=SAMPLERTYPE_Masks
   texture=/Game/Global/Textures/Default/Blanks/AllRed
  }
  NamedRerouteDeclaration(None=@n4.B) {name=SecondaryMask, nodeColor=<rgba 0 1 0.4141 1>}
  n5 = OneMinus(None=NamedRerouteUsage())
  n6 = Power(
   Base=Multiply(A=FunctionInput() {inputName=Fresnel, inputType=FunctionInput_Scalar, sortPriority=10}, B=Saturate(None=@n5))
   Exp=ScalarParameter() {defaultValue=2, group=HealedScar_Sand, parameterName=GhostFresnelPower, sortPriority=10}
  )
  n7 = Add(
   A=@n6
   B=Multiply(A=@n6, B=ScalarParameter() {defaultValue=-0.5, group=HealedScar_Sand, parameterName=Fresnel_GlowIRange, sortPriority=10}) {constB=1.3}
  )
  n8 = VectorParameter() {group=HealedScar_Sand, parameterName=SillhouetteGlowColor, sortPriority=10, defaultValue=<rgba 0.1858 0.9812 0.9911 1>}
  n9 = VectorParameter() {group=HealedScar_Sand, parameterName=WispsEmissiveColor_1, sortPriority=10, defaultValue=<rgba 0.031 0.522 0.75 1>}
  n10 = Multiply(A=@n9.RGB, B=NamedRerouteUsage())
  n11 = VectorParameter() {group=HealedScar_Sand, parameterName=WispsEmissiveColor_2, sortPriority=10, defaultValue=<rgba 0.1248 0.4423 1 1>}
  n12 = Multiply(A=@n11.RGB, B=NamedRerouteUsage())
  n13 = VectorParameter() {group=HealedScar_Sand, parameterName=WispsEmissiveColor_3, sortPriority=10, defaultValue=<rgba 0.4376 0.5293 0.5573 1>}
  n14 = NamedRerouteUsage()
  n15 = Multiply(A=@n9.RGB, B=NamedRerouteUsage())
  n16 = Multiply(A=@n11.RGB, B=NamedRerouteUsage())
  n17 = NamedRerouteUsage()
  n18 = LinearInterpolate(
   A=Add(A=Add(A=@n10, B=@n12), B=Multiply(A=@n13.RGB, B=@n14))
   B=Add(A=Add(A=@n15, B=@n16), B=Multiply(A=@n13.RGB, B=@n17))
   Alpha=NamedRerouteUsage()
  ) {desc=Float3}
  n19 = ScalarParameter() {defaultValue=0.2, group=HealedScar_Sand, parameterName=Wisps_Intensity}
  n20 = VectorParameter() {group=HealedScar_Sand, parameterName=ScarEmissiveColor, sortPriority=10, defaultValue=<rgba 0.1562 0.9682 1 1>}
  n21 = LinearInterpolate(
   A=Add(A=Multiply(A=@n7, B=@n8.RGB), B=Multiply(A=@n18, B=@n19))
   B=Multiply(A=ScalarParameter() {defaultValue=4, group=HealedScar_Sand, parameterName=EmissivePartIntensity}, B=@n20.RGB)
   Alpha=NamedRerouteUsage()
  )
  n22 = SphereMask(
   A=Add(A=NamedRerouteUsage(), B=Constant3Vector() {constant=<rgba 0 10 2 0>})
   B=Constant3Vector()
   Radius=ScalarParameter() {defaultValue=40, group=HealedScar_Sand, parameterName=SphereMask_Radius}
   Hardness=ScalarParameter() {defaultValue=0.3, group=HealedScar_Sand, parameterName=SphereMask_Hardness}
  ) {attenuationRadius=75, hardnessPercent=0.1}
  n23 = VectorParameter() {group=HealedScar_Sand, parameterName=SillhouetteGlowColor, sortPriority=10, defaultValue=<rgba 0.1858 0.9812 0.9911 1>}
  n24 = Multiply(A=Power(Base=InverseLinearInterpolate(Value=@n22) {constA=1, constB=0}), B=@n23.RGB)
  n25 = NamedRerouteUsage()
  n26 = StaticSwitchParameter(True=Add(A=@n21, B=Multiply(A=@n24, B=@n25)), False=@n21) {group=HealedScar_Sand, parameterName=UseSphereMaskGlow}
  n27 = NamedRerouteUsage()
  NamedRerouteDeclaration(None=StaticSwitchParameter(True=Add(A=@n26, B=@n27), False=@n26) {group=HealedScar_Sand, parameterName=UseEmissiveMap}) {name=Emissive, nodeColor=<rgba 0 0.9646 1 1>}
  NamedRerouteDeclaration(None=ScalarParameter() {defaultValue=1, group=HealedScar_Sand, parameterName=INAlpha}) {name=Alpha, nodeColor=<rgba 1 0.7969 0 1>}
  NamedRerouteDeclaration(None=VertexInterpolator(VS=PreSkinnedNormal()).PS) {name=PreskinnedNormal, nodeColor=<rgba 1 0.8672 0 1>}
  n28 = MaterialFunctionCall(Float3=RgbToHsv!(None=@n1.BaseColor)) {materialFunction=&BreakOutFloat3Components}
  n29 = Add(A=ScalarParameter() {group=HealedScar_Sand, parameterName=HUE_Offset}, B=@n28.R)
  n30 = Add(A=ScalarParameter() {group=HealedScar_Sand, parameterName=Saturation_Offset}, B=@n28.G)
  n31 = Add(A=@n28.B, B=ScalarParameter() {group=HealedScar_Sand, parameterName=Vaule_Offset})
  NamedRerouteDeclaration(None=HsvToRgb!(None=MaterialFunctionCall(X=@n29, Y=@n30, Z=@n31) {materialFunction=&MakeFloat3}.Result)) {name=BaseColor, nodeColor=<rgba 0.3125 1 0 1>}
  NamedRerouteDeclaration(None=ScalarParameter() {group=HealedScar_Sand, parameterName=GhostTransitionContrast, sortPriority=41}) {name=TransitionContrast, nodeColor=<rgba 0.2422 1 0 1>}
  NamedRerouteDeclaration(None=VertexInterpolator(VS=LocalPosition() {localOrigin=InstancePreSkinning}.XYZ).PS) {name=PreskinnedPos, nodeColor=<rgba 1 0.2109 0 1>}
  n32 = Add(A=NamedRerouteUsage(), B=NamedRerouteUsage())
  n33 = NamedRerouteUsage()
  n34 = Add(A=NamedRerouteUsage(), B=NamedRerouteUsage())
  n35 = NamedRerouteUsage()
  n36 = LinearInterpolate(
   A=LinearInterpolate(A=Add(A=@n32, B=@n33), B=Add(A=@n34, B=@n35), Alpha=NamedRerouteUsage()) {desc=Float3}
   B=NamedRerouteUsage()
   Alpha=NamedRerouteUsage()
  ) {desc=Float3}
  n37 = NamedRerouteUsage()
  n38 = Multiply(
   A=Fresnel(ExponentIn=ScalarParameter() {defaultValue=0.5, group=HealedScar_Sand, parameterName=Wisps_Normal_Exponent}, Normal=Transform(None=@n37)) {baseReflectFraction=0}
   B=ScalarParameter() {defaultValue=2, group=HealedScar_Sand, parameterName=Wisps_Fresnel_Strength}
  )
  NamedRerouteDeclaration(None=Saturate(None=Subtract(A=@n36, B=@n38))) {name=Wisps, nodeColor=<rgba 0 0.8702 1 1>}
  NamedRerouteDeclaration(None=@n4.G) {name=EmissiveMask, nodeColor=<rgba 0 1 0 1>}
  NamedRerouteDeclaration(None=NamedRerouteUsage()) {name=CachedFrontUV, nodeColor=<rgba 0.3359 0 1 1>}
  n39 = VectorParameter() {
   group=HealedScar_Sand
   parameterName=Wisps_Main_DistortUV
   channelNames {a=Offset.Y, b=Offset.X, g=Tilling.Y, r=Tilling.X}
   defaultValue=<rgba 0.5 0.2 0 0>
  }
  n40 = Add(
   A=Multiply(
    A=NamedRerouteUsage()
    B=MaterialFunctionCall(X=@n39.Tilling.X, Y=@n39.Tilling.X) {materialFunction=&MakeFloat2}.Result
   )
   B=MaterialFunctionCall(X=@n39.Offset.X, Y=@n39.Offset.X) {materialFunction=&MakeFloat2}.Result
  )
  n41 = VectorParameter() {group=HealedScar_Sand, parameterName=Wisps_Main_Flowmap_OffsetSpeed, defaultValue=<rgba -0.02 -0.02 0 0>}
  n42 = MaterialFunctionCall(X=@n41.R, Y=@n41.R) {materialFunction=&MakeFloat2}
  n43 = NamedRerouteUsage()
  n44 = ScalarParameter() {defaultValue=0.2, group=HealedScar_Sand, parameterName=FlowMap_Secondary_Scale}
  n45 = TextureSampleParameter2D(UVs=Add(A=Panner(Coordinate=@n40, Speed=@n42.Result), B=Multiply(A=@n43, B=@n44))) {
   desc=Only RG channel works
   group=HealedScar_Sand
   parameterName=Wisps_DistortUV_Flowmap
   samplerSource=SSM_Wrap_WorldGroupSettings
   texture=/BRCosmetics/Characters/Player/Male/Large/Bodies/M_LRG_HealedScar/Skins/Sand/Textures/T_HealedScar_Sand_Flowmap
  }
  n46 = Multiply(A=ComponentMask(None=@n45.RGB) {g=true, r=true}) {constB=2}
  NamedRerouteDeclaration(None=Multiply(A=Subtract(A=@n46), B=ScalarParameter() {defaultValue=0.5, group=HealedScar_Sand, parameterName=FlowMap_Main_Scale})) {name=FlowMap_Front_Main, nodeColor=<rgba 0 1 0.1797 1>}
  n47 = VectorParameter() {
   group=HealedScar_Sand
   parameterName=Wisps_Secondary_DistortUV
   channelNames {a=Offset.Y, b=Offset.X, g=Tilling.Y, r=Tilling.X}
   defaultValue=<rgba 1 0.5 2 0>
  }
  n48 = Multiply(
   A=NamedRerouteUsage()
   B=MaterialFunctionCall(X=@n47.Tilling.X, Y=@n47.Tilling.X) {materialFunction=&MakeFloat2}.Result
  )
  n49 = MaterialFunctionCall(X=@n47.Offset.X, Y=@n47.Offset.X) {materialFunction=&MakeFloat2}
  n50 = VectorParameter() {group=HealedScar_Sand, parameterName=Wisps_Secondary_Flowmap_OffsetSpeed, defaultValue=<rgba 0.01 -0.04 0 0>}
  n51 = TextureSampleParameter2D(UVs=Panner(
    Coordinate=Add(A=@n48, B=@n49.Result)
    Speed=MaterialFunctionCall(X=@n50.R, Y=@n50.R) {materialFunction=&MakeFloat2}.Result
   )) {
   desc=Only RG channel works
   group=HealedScar_Sand
   parameterName=Wisps_DistortUV_Flowmap
   samplerSource=SSM_Wrap_WorldGroupSettings
   texture=&T_HealedScar_Sand_Flowmap
  }
  n52 = ComponentMask(None=@n51.RGB) {g=true, r=true}
  NamedRerouteDeclaration(None=Subtract(A=Multiply(A=@n52) {constB=2})) {name=FlowMap_Front_Secondary, nodeColor=<rgba 0 1 0.2969 1>}
  n53 = NamedRerouteUsage()
  n54 = VectorParameter() {
   group=HealedScar_Sand
   parameterName=WispsNoise1_UV
   channelNames {a=Offset Y, b=Offset X, g=Tilling Y, r=Tilling X}
   defaultValue=<rgba 0.75 0.75 0 0>
  }
  n55 = Multiply(
   A=@n53
   B=MaterialFunctionCall(X=@n54.Tilling X, Y=@n54.Tilling X) {materialFunction=&MakeFloat2}.Result
  )
  n56 = ComponentMask(None=MaterialFunctionCall(X=@n54.Offset X, Y=@n54.Offset X) {materialFunction=&MakeFloat2}.Result) {g=true, r=true}
  n57 = VectorParameter() {group=HealedScar_Sand, parameterName=WispsNoise1_OffsetSpeed, defaultValue=<rgba 0.01 0.02 0 0>}
  n58 = Add(
   A=Panner(
    Coordinate=Add(A=@n55, B=@n56)
    Speed=MaterialFunctionCall(X=@n57.R, Y=@n57.R) {materialFunction=&MakeFloat2}.Result
   )
   B=NamedRerouteUsage()
  )
  n59 = ScalarParameter() {group=HealedScar_Sand, parameterName=Wisps1_Height}
  n60 = TextureObjectParameter() {
   group=HealedScar_Sand
   parameterName=WispsNoise_Texture
   sortPriority=20
   texture=/Game/Global/Textures/Default/Tileable_Patterns/T_Blowing_Smoke_Pattern2_Tile_DISP
  }
  NamedRerouteDeclaration(None=TextureSample(UVs=BumpOffset(Coordinate=@n58, Height=@n59), Tex=@n60) {desc=Front, samplerSource=SSM_Wrap_WorldGroupSettings}.R) {name=Front_Wisps_01, nodeColor=<rgba 0 0.9531 1 1>}
  n61 = VectorParameter() {
   group=HealedScar_Sand
   parameterName=WispsNoise2_UV
   channelNames {a=Offset Y, b=Offset X, g=Tilling Y, r=Tilling X}
   defaultValue=<rgba 1 1 0 0>
  }
  n62 = Multiply(
   A=@n53
   B=MaterialFunctionCall(X=@n61.Tilling X, Y=@n61.Tilling X) {materialFunction=&MakeFloat2}.Result
  )
  n63 = ComponentMask(None=MaterialFunctionCall(X=@n61.Offset X, Y=@n61.Offset X) {materialFunction=&MakeFloat2}.Result) {g=true, r=true}
  n64 = VectorParameter() {group=HealedScar_Sand, parameterName=WispsNoise2_OffsetSpeed, defaultValue=<rgba 0 0.01 0 0>}
  n65 = Add(
   A=Panner(
    Coordinate=Add(A=@n62, B=@n63)
    Speed=MaterialFunctionCall(X=@n64.R, Y=@n64.R) {materialFunction=&MakeFloat2}.Result
   )
   B=NamedRerouteUsage()
  )
  n66 = ScalarParameter() {defaultValue=-1.5, group=HealedScar_Sand, parameterName=Wisps2_Height}
  NamedRerouteDeclaration(None=TextureSample(UVs=BumpOffset(Coordinate=@n65, Height=@n66), Tex=@n60) {desc=Front, samplerSource=SSM_Wrap_WorldGroupSettings}.R) {name=Front_Wisps_02, nodeColor=<rgba 0 0.9531 1 1>}
  n67 = VectorParameter() {
   group=HealedScar_Sand
   parameterName=WispsNoise3_UV
   channelNames {a=Offset Y, b=Offset X, g=Tilling Y, r=Tilling X}
   defaultValue=<rgba 1.25 1.25 0 0>
  }
  n68 = Multiply(
   A=@n53
   B=MaterialFunctionCall(X=@n67.Tilling X, Y=@n67.Tilling X) {materialFunction=&MakeFloat2}.Result
  )
  n69 = ComponentMask(None=MaterialFunctionCall(X=@n67.Offset X, Y=@n67.Offset X) {materialFunction=&MakeFloat2}.Result) {g=true, r=true}
  n70 = VectorParameter() {group=HealedScar_Sand, parameterName=WispsNoise3_OffsetSpeed, defaultValue=<rgba 0 0.05 0 0>}
  n71 = Add(
   A=Panner(
    Coordinate=Add(A=@n68, B=@n69)
    Speed=MaterialFunctionCall(X=@n70.R, Y=@n70.R) {materialFunction=&MakeFloat2}.Result
   )
   B=NamedRerouteUsage()
  )
  n72 = ScalarParameter() {defaultValue=-3, group=HealedScar_Sand, parameterName=Wisps3_Height}
  NamedRerouteDeclaration(None=TextureSample(UVs=BumpOffset(Coordinate=@n71, Height=@n72), Tex=@n60) {desc=Front, samplerSource=SSM_Wrap_WorldGroupSettings}.R) {name=Front_Wisps_03, nodeColor=<rgba 0 0.9531 1 1>}
  n73 = NamedRerouteUsage()
  n74 = NamedRerouteUsage()
  n75 = NamedRerouteUsage()
  NamedRerouteDeclaration(None=MaterialFunctionCall!(XY vector=LinearInterpolate(A=@n73, B=@n74, Alpha=@n75), Z sign, Default Normalization vector) {materialFunction=/Engine/Functions/Engine_MaterialFunctions02/Math/DeriveNormalZ_Function}.Result) {name=Wisps_Normal, nodeColor=<rgba 0 1 0.8594 1>}
  n76 = Time()
  n77 = VectorParameter() {group=HealedScar_Sand, parameterName=Wisps_Speed, defaultValue=<rgba 0 0 1 0>}
  n78 = ScalarParameter() {defaultValue=64, group=HealedScar_Sand, parameterName=WispsTextureSize}
  n79 = Divide(A=Add(A=NamedRerouteUsage(), B=Multiply(A=@n76, B=@n77.RGB)), B=Multiply(A=Abs(None=@n78)) {constB=-1})
  NamedRerouteDeclaration(None=Multiply(A=ComponentMask(None=@n79) {b=true, r=true}, B=Constant2Vector() {g=1, r=-1})) {name=Triplanar_Front_UV, nodeColor=<rgba 0 0.1328 1 1>}
  NamedRerouteDeclaration(None=ComponentMask(None=@n79) {b=true, g=true}) {name=Triplanar_Side_UV, nodeColor=<rgba 0 0.1328 1 1>}
  NamedRerouteDeclaration(None=ComponentMask(None=@n79) {g=true, r=true}) {name=Triplanar_Top_UV, nodeColor=<rgba 0 0.1328 1 1>}
  NamedRerouteDeclaration(None=NamedRerouteUsage()) {name=CachedSideUV, nodeColor=<rgba 0.3359 0 1 1>}
  n80 = VectorParameter() {
   group=HealedScar_Sand
   parameterName=Wisps_Main_DistortUV
   channelNames {a=Offset.Y, b=Offset.X, g=Tilling.Y, r=Tilling.X}
   defaultValue=<rgba 1 1 0 0>
  }
  n81 = Add(
   A=Multiply(
    A=NamedRerouteUsage()
    B=MaterialFunctionCall(X=@n80.Tilling.X, Y=@n80.Tilling.X) {materialFunction=&MakeFloat2}.Result
   )
   B=MaterialFunctionCall(X=@n80.Offset.X, Y=@n80.Offset.X) {materialFunction=&MakeFloat2}.Result
  )
  n82 = VectorParameter() {group=HealedScar_Sand, parameterName=Wisps_Main_Flowmap_OffsetSpeed, defaultValue=<rgba -0.02 -0.02 0 0>}
  n83 = MaterialFunctionCall(X=@n82.R, Y=@n82.R) {materialFunction=&MakeFloat2}
  n84 = NamedRerouteUsage()
  n85 = ScalarParameter() {group=HealedScar_Sand, parameterName=FlowMap_Secondary_Scale}
  n86 = TextureSampleParameter2D(UVs=Add(A=Panner(Coordinate=@n81, Speed=@n83.Result), B=Multiply(A=@n84, B=@n85))) {
   desc=Only RG channel works
   group=HealedScar_Sand
   parameterName=Wisps_DistortUV_Flowmap
   samplerSource=SSM_Wrap_WorldGroupSettings
   texture=&T_HealedScar_Sand_Flowmap
  }
  n87 = Multiply(A=ComponentMask(None=@n86.RGB) {g=true, r=true}) {constB=2}
  NamedRerouteDeclaration(None=Multiply(A=Subtract(A=@n87), B=ScalarParameter() {group=HealedScar_Sand, parameterName=FlowMap_Main_Scale})) {name=FlowMap_Side_Main, nodeColor=<rgba 0 1 0.1797 1>}
  n88 = VectorParameter() {
   group=HealedScar_Sand
   parameterName=Wisps_Secondary_DistortUV
   channelNames {a=Offset.Y, b=Offset.X, g=Tilling.Y, r=Tilling.X}
   defaultValue=<rgba 1 1 2 0>
  }
  n89 = Multiply(
   A=NamedRerouteUsage()
   B=MaterialFunctionCall(X=@n88.Tilling.X, Y=@n88.Tilling.X) {materialFunction=&MakeFloat2}.Result
  )
  n90 = MaterialFunctionCall(X=@n88.Offset.X, Y=@n88.Offset.X) {materialFunction=&MakeFloat2}
  n91 = VectorParameter() {group=HealedScar_Sand, parameterName=Wisps_Secondary_Flowmap_OffsetSpeed, defaultValue=<rgba 0.01 -0.04 0 0>}
  n92 = TextureSampleParameter2D(UVs=Panner(
    Coordinate=Add(A=@n89, B=@n90.Result)
    Speed=MaterialFunctionCall(X=@n91.R, Y=@n91.R) {materialFunction=&MakeFloat2}.Result
   )) {
   desc=Only RG channel works
   group=HealedScar_Sand
   parameterName=Wisps_DistortUV_Flowmap
   samplerSource=SSM_Wrap_WorldGroupSettings
   texture=&T_HealedScar_Sand_Flowmap
  }
  n93 = ComponentMask(None=@n92.RGB) {g=true, r=true}
  NamedRerouteDeclaration(None=Subtract(A=Multiply(A=@n93) {constB=2})) {name=FlowMap_Side_Secondary, nodeColor=<rgba 0 1 0.2969 1>}
  n94 = NamedRerouteUsage()
  n95 = VectorParameter() {
   group=HealedScar_Sand
   parameterName=WispsNoise1_UV
   channelNames {a=Offset Y, b=Offset X, g=Tilling Y, r=Tilling X}
   defaultValue=<rgba 0.5 0.5 0 0>
  }
  n96 = Multiply(
   A=@n94
   B=MaterialFunctionCall(X=@n95.Tilling X, Y=@n95.Tilling X) {materialFunction=&MakeFloat2}.Result
  )
  n97 = ComponentMask(None=MaterialFunctionCall(X=@n95.Offset X, Y=@n95.Offset X) {materialFunction=&MakeFloat2}.Result) {g=true, r=true}
  n98 = VectorParameter() {group=HealedScar_Sand, parameterName=WispsNoise1_OffsetSpeed, defaultValue=<rgba 0 0.02 0 0>}
  n99 = Add(
   A=Panner(
    Coordinate=Add(A=@n96, B=@n97)
    Speed=MaterialFunctionCall(X=@n98.R, Y=@n98.R) {materialFunction=&MakeFloat2}.Result
   )
   B=NamedRerouteUsage()
  )
  n100 = ScalarParameter() {group=HealedScar_Sand, parameterName=Wisps1_Height}
  NamedRerouteDeclaration(None=TextureSample(UVs=BumpOffset(Coordinate=@n99, Height=@n100), Tex=@n60) {desc=Front, samplerSource=SSM_Wrap_WorldGroupSettings}.R) {name=Side_Wisps_01, nodeColor=<rgba 0 0.9531 1 1>}
  n101 = VectorParameter() {
   group=HealedScar_Sand
   parameterName=WispsNoise2_UV
   channelNames {a=Offset Y, b=Offset X, g=Tilling Y, r=Tilling X}
   defaultValue=<rgba 1 1 -0.528 0.2>
  }
  n102 = Multiply(
   A=@n94
   B=MaterialFunctionCall(X=@n101.Tilling X, Y=@n101.Tilling X) {materialFunction=&MakeFloat2}.Result
  )
  n103 = ComponentMask(None=MaterialFunctionCall(X=@n101.Offset X, Y=@n101.Offset X) {materialFunction=&MakeFloat2}.Result) {g=true, r=true}
  n104 = VectorParameter() {group=HealedScar_Sand, parameterName=WispsNoise2_OffsetSpeed, defaultValue=<rgba 0 0.05 0 0>}
  n105 = Add(
   A=Panner(
    Coordinate=Add(A=@n102, B=@n103)
    Speed=MaterialFunctionCall(X=@n104.R, Y=@n104.R) {materialFunction=&MakeFloat2}.Result
   )
   B=NamedRerouteUsage()
  )
  n106 = ScalarParameter() {defaultValue=-1.5, group=HealedScar_Sand, parameterName=Wisps2_Height}
  NamedRerouteDeclaration(None=TextureSample(UVs=BumpOffset(Coordinate=@n105, Height=@n106), Tex=@n60) {desc=Front, samplerSource=SSM_Wrap_WorldGroupSettings}.R) {name=Side_Wisps_02, nodeColor=<rgba 0 0.9531 1 1>}
  n107 = VectorParameter() {
   group=HealedScar_Sand
   parameterName=WispsNoise3_UV
   channelNames {a=Offset Y, b=Offset X, g=Tilling Y, r=Tilling X}
   defaultValue=<rgba 2 2 0.2 0.4>
  }
  n108 = Multiply(
   A=@n94
   B=MaterialFunctionCall(X=@n107.Tilling X, Y=@n107.Tilling X) {materialFunction=&MakeFloat2}.Result
  )
  n109 = ComponentMask(None=MaterialFunctionCall(X=@n107.Offset X, Y=@n107.Offset X) {materialFunction=&MakeFloat2}.Result) {g=true, r=true}
  n110 = VectorParameter() {group=HealedScar_Sand, parameterName=WispsNoise3_OffsetSpeed, defaultValue=<rgba 0 0.05 0 0>}
  n111 = Add(
   A=Panner(
    Coordinate=Add(A=@n108, B=@n109)
    Speed=MaterialFunctionCall(X=@n110.R, Y=@n110.R) {materialFunction=&MakeFloat2}.Result
   )
   B=NamedRerouteUsage()
  )
  n112 = ScalarParameter() {defaultValue=-3, group=HealedScar_Sand, parameterName=Wisps3_Height}
  NamedRerouteDeclaration(None=TextureSample(UVs=BumpOffset(Coordinate=@n111, Height=@n112), Tex=@n60) {desc=Front, samplerSource=SSM_Wrap_WorldGroupSettings}.R) {name=Side_Wisps_03, nodeColor=<rgba 0 0.9531 1 1>}
  n113 = NamedRerouteUsage()
  n114 = Saturate(None=ComponentMask(None=@n113) {g=true})
  n115 = Saturate(None=ComponentMask(None=Multiply(A=@n113) {constB=-1}) {g=true})
  NamedRerouteDeclaration(None=SmoothStep(Value=Add(A=@n114, B=@n115)) {constMin=0.4}) {name=SideFrontBlend, nodeColor=<rgba 1 0.4219 0 1>}
  n116 = Saturate(None=ComponentMask(None=@n113) {b=true})
  n117 = Saturate(None=ComponentMask(None=Multiply(A=@n113) {constB=-1}) {b=true})
  NamedRerouteDeclaration(None=SmoothStep(Value=Add(A=@n116, B=@n117)) {constMin=0.4}) {name=TopBlend, nodeColor=<rgba 1 0.4219 0 1>}
  NamedRerouteDeclaration(None=TextureSample(UVs=NamedRerouteUsage(), Tex=@n60) {desc=Top, samplerSource=SSM_Wrap_WorldGroupSettings}.R) {name=Top_Wisps, nodeColor=<rgba 0 0.9531 1 1>}
  NamedRerouteDeclaration(None=@n4.R) {name=MainMask, nodeColor=<rgba 0 0 0.991 1>}
  NamedRerouteDeclaration(None=@n1.EmissiveColor) {name=OriginalEmissiveColor, nodeColor=<rgba 1 0 0.9141 1>}
 ]}
